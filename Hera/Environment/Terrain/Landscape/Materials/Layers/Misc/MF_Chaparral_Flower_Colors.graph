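# Material function for the chaparral landscape layer that outputs a flower color parameter, a world-space flower mask and a layer noise mask.
# The flower mask multiplies two world-UV samples of one noise texture, one negated and one offset, then sharpens it with power and scale/bias parameters.
# The layer mask combines two blue-channel noise scales.
# Raw noise is also output; DEBUGOUT is constant, and the BeersLaw and variation-mask branches reach no output.
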MaterialFunction /Hera/Environment/Terrain/Landscape/Materials/Layers/Misc/MF_Chaparral_Flower_Colors
graph {expressions {
  MaterialExpressionAdd_2 {
   class=MaterialExpressionAdd
   inputs [
    A {from=MaterialExpressionMaterialFunctionCall_4, output=Result}
    B {from=MaterialExpressionConvert_0}
   ]
  }
  MaterialExpressionAdd_4 {
   class=MaterialExpressionAdd
   inputs [
    A {from=MaterialExpressionMultiply_24}
    B {from=MaterialExpressionScalarParameter_25}
   ]
  }
  MaterialExpressionConstant4Vector_0 {class=MaterialExpressionConstant4Vector, inputs [], properties {constant=<rgba 0.2903 0.2856 0.5885 1>}}
  MaterialExpressionConstant4Vector_1 {class=MaterialExpressionConstant4Vector, inputs [], properties {constant=<rgba 0.1276 0.12 0.6 1>}}
  MaterialExpressionConstant_0 {class=MaterialExpressionConstant, properties {r=10}}
  MaterialExpressionConstant_1 {class=MaterialExpressionConstant, properties {r=0.03}}
  MaterialExpressionConstant_10 {class=MaterialExpressionConstant, properties {r=720}}
  MaterialExpressionConstant_11 {class=MaterialExpressionConstant, properties {r=10}}
  MaterialExpressionConstant_12 {class=MaterialExpressionConstant, properties {r=5000}}
  MaterialExpressionConstant_13 {class=MaterialExpressionConstant, properties {r=3300}}
  MaterialExpressionConstant_14 {class=MaterialExpressionConstant, properties {r=0.33}}
  MaterialExpressionConstant_15 {class=MaterialExpressionConstant, properties {r=0.66}}
  MaterialExpressionConstant_2 {class=MaterialExpressionConstant, properties {r=2}}
  MaterialExpressionConstant_3 {class=MaterialExpressionConstant, properties {r=0.24}}
  MaterialExpressionConstant_4 {class=MaterialExpressionConstant, properties {r=4000}}
  MaterialExpressionConstant_5 {class=MaterialExpressionConstant, properties {r=2300}}
  MaterialExpressionConstant_6 {class=MaterialExpressionConstant, properties {r=2300}}
  MaterialExpressionConstant_7 {class=MaterialExpressionConstant, properties {r=3}}
  MaterialExpressionConstant_8 {class=MaterialExpressionConstant, properties {r=1}}
  MaterialExpressionConstant_9 {class=MaterialExpressionConstant, properties {r=3000}}
  MaterialExpressionConvert_0 {
   class=MaterialExpressionConvert
   inputs [
    None {from=MaterialExpressionScalarParameter_19}
    None {from=MaterialExpressionScalarParameter_20}
   ]
   properties {
    nodeName=Make Vector2
    convertInputs [
     {
      type=Scalar
      expressionInput {
       expression=&self:MaterialExpressionScalarParameter_19
       inputName=None
       mask=0
       maskA=0
       maskB=0
       maskG=0
       maskR=0
       outputIndex=0
      }
     }
     {
      type=Scalar
      expressionInput {
       expression=&self:MaterialExpressionScalarParameter_20
       inputName=None
       mask=0
       maskA=0
       maskB=0
       maskG=0
       maskR=0
       outputIndex=0
      }
     }
    ]
    convertOutputs [{type=Vector2}]
   }
  }
  MaterialExpressionFunctionOutput_0 {
   class=MaterialExpressionFunctionOutput
   inputs [None {from=MaterialExpressionOneMinus_13, output=""}]
   properties {outputName=LandscapeLayerMaskNoise, sortPriority=-500}
  }
  MaterialExpressionFunctionOutput_1 {
   class=MaterialExpressionFunctionOutput
   inputs [None {from=MaterialExpressionSaturate_7, output=""}]
   properties {outputName=FlowerWorldMask1, sortPriority=-400}
  }
  MaterialExpressionFunctionOutput_2 {
   class=MaterialExpressionFunctionOutput
   inputs [None {from=MaterialExpressionVectorParameter_0, output=RGB}]
   properties {outputName=Flower1_Color}
  }
  MaterialExpressionFunctionOutput_3 {
   class=MaterialExpressionFunctionOutput
   inputs [None {from=MaterialExpressionConstant_8, output=""}]
   properties {outputName=DEBUGOUT, sortPriority=500}
  }
  MaterialExpressionFunctionOutput_4 {
   class=MaterialExpressionFunctionOutput
   inputs [None {from=MaterialExpressionTextureSample_16, output=RGB}]
   properties {outputName=Noise01_RGB_RAW, sortPriority=400}
  }
  MaterialExpressionMaterialFunctionCall_0 {
   class=MaterialExpressionMaterialFunctionCall
   inputs [
    Coordinates {from=""}
    R_01_Scale {from=MaterialExpressionConstant_4}
    R_02_Scale {from=MaterialExpressionConstant_5}
    R_03_Scale {from=MaterialExpressionConstant_6}
    R_04_Scale {from=""}
    G_01_Scale {from=""}
    G_02_Scale {from=""}
    G_03_Scale {from=""}
    G_04_Scale {from=""}
   ]
   properties {materialFunction=/Game/Athena/Helios/Environments/Landscape/Materials/MaterialFunctions_Utilities/MF_VariationMasks2}
  }
  MaterialExpressionMaterialFunctionCall_1 {
   class=MaterialExpressionMaterialFunctionCall
   inputs [Scale {from=MaterialExpressionScalarParameter_49}]
   properties {materialFunction=/Game/Athena/Apollo/Environments/Landscape/Materials/Biome/CommonMaterials/MF_SimpleWorldUV}
  }
  MaterialExpressionMaterialFunctionCall_11 {
   class=MaterialExpressionMaterialFunctionCall
   inputs [
    Thickness {from=MaterialExpressionSquareRoot_2}
    Depth Scale {from=MaterialExpressionConstant_7}
   ]
   properties {materialFunction=/Engine/Functions/Engine_MaterialFunctions01/Volumetrics/BeersLaw}
  }
  MaterialExpressionMaterialFunctionCall_3 {
   class=MaterialExpressionMaterialFunctionCall
   inputs [Scale {from=MaterialExpressionScalarParameter_17}]
   properties {materialFunction=&MF_SimpleWorldUV}
  }
  MaterialExpressionMaterialFunctionCall_4 {
   class=MaterialExpressionMaterialFunctionCall
   inputs [Scale {from=MaterialExpressionScalarParameter_18}]
   properties {materialFunction=&MF_SimpleWorldUV}
  }
  MaterialExpressionMaterialFunctionCall_9 {
   class=MaterialExpressionMaterialFunctionCall
   inputs [Scale {from=MaterialExpressionScalarParameter_50}]
   properties {materialFunction=&MF_SimpleWorldUV}
  }
  MaterialExpressionMultiply_0 {
   class=MaterialExpressionMultiply
   inputs [
    A {from=MaterialExpressionTextureSample_16, output=B}
    B {from=MaterialExpressionTextureSample_17, output=B}
   ]
  }
  MaterialExpressionMultiply_1 {
   class=MaterialExpressionMultiply
   inputs [
    A {from=MaterialExpressionPower_0}
    B {from=MaterialExpressionConstant_2}
   ]
  }
  MaterialExpressionMultiply_12 {
   class=MaterialExpressionMultiply
   inputs [
    A {from=MaterialExpressionOneMinus_10}
    B {from=""}
   ]
   properties {constB=1.25}
  }
  MaterialExpressionMultiply_13 {
   class=MaterialExpressionMultiply
   inputs [
    A {from=MaterialExpressionOneMinus_12}
    B {from=MaterialExpressionSaturate_2}
   ]
  }
  MaterialExpressionMultiply_17 {
   class=MaterialExpressionMultiply
   inputs [
    A {from=MaterialExpressionMaterialFunctionCall_3, output=Result}
    B {from=""}
   ]
   properties {constB=-1}
  }
  MaterialExpressionMultiply_19 {
   class=MaterialExpressionMultiply
   inputs [
    A {from=MaterialExpressionTextureSample_4, output=R}
    B {from=MaterialExpressionTextureSample_5, output=R}
   ]
  }
  MaterialExpressionMultiply_24 {
   class=MaterialExpressionMultiply
   inputs [
    A {from=MaterialExpressionOneMinus_1}
    B {from=MaterialExpressionScalarParameter_24}
   ]
  }
  MaterialExpressionMultiply_52 {
   class=MaterialExpressionMultiply
   inputs [
    A {from=MaterialExpressionTextureSample_16, output=B}
    B {from=MaterialExpressionTextureSample_17, output=B}
   ]
  }
  MaterialExpressionMultiply_56 {
   class=MaterialExpressionMultiply
   inputs [
    A {from=MaterialExpressionOneMinus_11}
    B {from=MaterialExpressionMaterialFunctionCall_11, output=Result}
   ]
  }
  MaterialExpressionMultiply_57 {
   class=MaterialExpressionMultiply
   inputs [
    A {from=MaterialExpressionTextureCoordinate_8}
    B {from=MaterialExpressionConstant_1}
   ]
  }
  MaterialExpressionNamedRerouteDeclaration_0 {
   class=MaterialExpressionNamedRerouteDeclaration
   inputs [None {from=MaterialExpressionMaterialFunctionCall_0, output=R_02}]
   properties {name=Chaparral_Noise2, nodeColor=<rgba 0.5 0 1 1>}
  }
  MaterialExpressionOneMinus_0 {class=MaterialExpressionOneMinus, inputs [None {from=MaterialExpressionMultiply_19}]}
  MaterialExpressionOneMinus_1 {class=MaterialExpressionOneMinus, inputs [None {from=MaterialExpressionSaturate_0}]}
  MaterialExpressionOneMinus_10 {class=MaterialExpressionOneMinus, inputs [None {from=MaterialExpressionPower_6}]}
  MaterialExpressionOneMinus_11 {class=MaterialExpressionOneMinus, inputs [None {from=MaterialExpressionTextureSample_16, output=B}]}
  MaterialExpressionOneMinus_12 {class=MaterialExpressionOneMinus, inputs [None {from=MaterialExpressionTextureSample_16, output=B}]}
  MaterialExpressionOneMinus_13 {class=MaterialExpressionOneMinus, inputs [None {from=MaterialExpressionMultiply_13}]}
  MaterialExpressionOneMinus_4 {class=MaterialExpressionOneMinus, inputs [None {from=MaterialExpressionNamedRerouteDeclaration_0}]}
  MaterialExpressionOneMinus_5 {class=MaterialExpressionOneMinus, inputs [None {from=MaterialExpressionMultiply_56}]}
  MaterialExpressionPower_0 {
   class=MaterialExpressionPower
   inputs [
    Base {from=MaterialExpressionOneMinus_0}
    Exp {from=MaterialExpressionScalarParameter_21}
   ]
   properties {constExponent=4}
  }
  MaterialExpressionPower_2 {
   class=MaterialExpressionPower
   inputs [
    Base {from=MaterialExpressionOneMinus_4}
    Exp {from=MaterialExpressionConstant_0}
   ]
   properties {constExponent=4}
  }
  MaterialExpressionPower_6 {
   class=MaterialExpressionPower
   inputs [
    Base {from=MaterialExpressionMultiply_0}
    Exp {from=MaterialExpressionConstant_3}
   ]
   properties {constExponent=0.2}
  }
  MaterialExpressionSaturate_0 {class=MaterialExpressionSaturate, inputs [None {from=MaterialExpressionMultiply_1}]}
  MaterialExpressionSaturate_2 {class=MaterialExpressionSaturate, inputs [None {from=MaterialExpressionMultiply_12}]}
  MaterialExpressionSaturate_4 {class=MaterialExpressionSaturate, inputs [None {from=MaterialExpressionOneMinus_5}]}
  MaterialExpressionSaturate_7 {class=MaterialExpressionSaturate, inputs [None {from=MaterialExpressionAdd_4}]}
  MaterialExpressionScalarParameter_17 {
   class=MaterialExpressionScalarParameter
   properties {defaultValue=7000, group=Grass_Chaparral_Flower, parameterName=Grass_Chaparral_FlowerWM_Flower1Scale}
  }
  MaterialExpressionScalarParameter_18 {
   class=MaterialExpressionScalarParameter
   properties {defaultValue=4300, group=Grass_Chaparral_Flower, parameterName=Grass_Chaparral_FlowerWM_Flower2Scale}
  }
  MaterialExpressionScalarParameter_19 {
   class=MaterialExpressionScalarParameter
   properties {defaultValue=0.33, group=Grass_Chaparral_Flower, parameterName=Grass_Chaparral_FlowerWM_Flower2Offset1}
  }
  MaterialExpressionScalarParameter_20 {
   class=MaterialExpressionScalarParameter
   properties {defaultValue=0.66, group=Grass_Chaparral_Flower, parameterName=Grass_Chaparral_FlowerWM_Flower2Offset2}
  }
  MaterialExpressionScalarParameter_21 {
   class=MaterialExpressionScalarParameter
   properties {defaultValue=6, group=Grass_Chaparral_Flower, parameterName=Grass_Chaparral_FlowerWM_Flower_Pow}
  }
  MaterialExpressionScalarParameter_24 {
   class=MaterialExpressionScalarParameter
   properties {defaultValue=10, group=Grass_Chaparral_Flower, parameterName=Grass_Chaparral_FlowerWM_Flower_Mult}
  }
  MaterialExpressionScalarParameter_25 {class=MaterialExpressionScalarParameter, properties {group=Grass_Chaparral_Flower, parameterName=Grass_Chaparral_FlowerWM_Flower_Add}}
  MaterialExpressionScalarParameter_49 {
   class=MaterialExpressionScalarParameter
   properties {defaultValue=3000, group=Grass_Chaparral_Flower, parameterName=Grass_Chaparral_Flower_Noise01_Scale}
  }
  MaterialExpressionScalarParameter_50 {
   class=MaterialExpressionScalarParameter
   properties {defaultValue=720, group=Grass_Chaparral_Flower, parameterName=Grass_Chaparral_Flower_Noise02_Scale}
  }
  MaterialExpressionSquareRoot_2 {class=MaterialExpressionSquareRoot, inputs [None {from=MaterialExpressionMultiply_52, output=""}], properties {}}
  MaterialExpressionTextureCoordinate_8 {class=MaterialExpressionTextureCoordinate}
  MaterialExpressionTextureObject_1 {class=MaterialExpressionTextureObject, inputs [], properties {texture=/Game/Environments/Helios/Terrain/Textures/T_Helios_Landscape_Noise}}
  MaterialExpressionTextureSample_16 {
   class=MaterialExpressionTextureSample
   inputs [
    UVs {from=MaterialExpressionMaterialFunctionCall_1, output=Result}
    Tex {from=MaterialExpressionTextureObject_1}
    Apply View MipBias {from=""}
   ]
   properties {samplerSource=SSM_Wrap_WorldGroupSettings, texture=&T_Helios_Landscape_Noise}
  }
  MaterialExpressionTextureSample_17 {
   class=MaterialExpressionTextureSample
   inputs [
    UVs {from=MaterialExpressionMaterialFunctionCall_9, output=Result}
    Tex {from=MaterialExpressionTextureObject_1}
    Apply View MipBias {from=""}
   ]
   properties {samplerSource=SSM_Wrap_WorldGroupSettings, texture=&T_Helios_Landscape_Noise}
  }
  MaterialExpressionTextureSample_4 {
   class=MaterialExpressionTextureSample
   inputs [
    UVs {from=MaterialExpressionMultiply_17}
    Tex {from=MaterialExpressionTextureObject_1}
    Apply View MipBias {from=""}
   ]
   properties {samplerSource=SSM_Wrap_WorldGroupSettings, texture=&T_Helios_Landscape_Noise}
  }
  MaterialExpressionTextureSample_5 {
   class=MaterialExpressionTextureSample
   inputs [
    UVs {from=MaterialExpressionAdd_2}
    Tex {from=MaterialExpressionTextureObject_1}
    Apply View MipBias {from=""}
   ]
   properties {samplerSource=SSM_Wrap_WorldGroupSettings, texture=&T_Helios_Landscape_Noise}
  }
  MaterialExpressionVectorParameter_0 {
   class=MaterialExpressionVectorParameter
   properties {group=Grass_Chaparral_Flower, parameterName=Grass_Chaparral_FlowerWM_FlowerColor1, defaultValue=<rgba 0.2719 0.24 0.6 1>}
  }
 }}
