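# Landscape material function for open desert macro detail, outputting four-band sediment layer values, slope-masked normals and unmodified normals.
# A world-projected dune-wave normal texture and periodic position-based noise distort world Z, which mobile uses raw, before per-layer height and falloff form SedimentLayers.
# Normals are weighted by a vertex-normal slope mask times BiomeMask, with a basis transform declaring corrected world and tangent normals.
# Several debug and static-mesh branches are left unconnected.
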MaterialFunction /Hera/Environment/Terrain/Landscape/Materials/Layers/Misc/MF_Desert_MacroDetail
graph {nodes [
  n1 = VectorParameter() {group=OpenDesert_MacroDetail, parameterName=OD_MD_DistortionDirection, defaultValue=<rgba -150 150 0 0>}
  n2 = FunctionOutput(None=NamedRerouteUsage()) {outputName=UnmodifiedNormals, sortPriority=-99}
  n3 = Multiply(A=@n1.RGB, B=@n2)
  ComponentMask(None=@n3) {r=true}
  ComponentMask(None=@n3) {g=true}
  Constant4Vector() {constant=<rgba -600 500 -1200 600>}
  Constant4Vector() {constant=<rgba -2000 800 -2400 600>}
  n4 = WorldPosition()
  n5 = VectorParameter() {
   group=OpenDesert_MacroDetail
   parameterName=OD_MD_NormalSettings
   channelNames {a=Normal_SlopeMaskAdd, b=Normal_SlopeMaskMod, g=Normal_SlopeMaskPower, r=NormalScale}
   defaultValue=<rgba 8500 50 0.35 0.15>
  }
  n6 = TextureSample(UVs=MaterialFunctionCall!(Scale=@n5.NormalScale) {materialFunction=/Game/Athena/Apollo/Environments/Landscape/Materials/Biome/CommonMaterials/MF_SimpleWorldUV}.Result) {
   samplerSource=SSM_Wrap_WorldGroupSettings
   samplerType=SAMPLERTYPE_Normal
   texture=/Game/Environments/Helios/Terrain/Textures/T_Helios_DesertWaves_N
  }
  n7 = Add(A=Add(A=@n4.Z, B=Multiply(A=@n1.R, B=@n6.R)), B=Multiply(A=@n1.G, B=@n6.G))
  n8 = Frac(None=Divide(A=@n4.XYZ, B=ScalarParameter() {defaultValue=3500, group=OpenDesert_MacroDetail, parameterName=OD_MD_SineWaveNoise_Scale, sortPriority=501}))
  n9 = Multiply(A=OneMinus(None=@n8), B=@n8)
  n10 = MaterialFunctionCall(Color=Multiply(A=@n9, B=@n9)) {materialFunction=&SplitComponents}
  n11 = Add(A=@n10.R, B=@n10.R)
  n12 = Subtract(
   A=@n7
   B=Multiply(
    A=Add(A=@n11, B=@n10.B)
    B=ScalarParameter() {defaultValue=750, group=OpenDesert_MacroDetail, parameterName=OD_MD_SineWaveNoise_Mod, sortPriority=501}
   )
  )
  n13 = ShadingPathSwitch(Default=@n12, Mobile=@n4.Z)
  n14 = VectorParameter() {
   group=OpenDesert_MacroDetail
   parameterName=OD_MD_Layer1_2_Settings
   channelNames {a=Layer2_Falloff, b=Layer2_Height, g=Layer1_Falloff, r=Layer1_Height}
   defaultValue=<rgba -100 550 -600 800>
  }
  n15 = Add(A=@n13, B=@n14.Layer1_Height)
  n16 = Add(A=@n13, B=@n14.Layer2_Height)
  n17 = VectorParameter() {
   group=OpenDesert_MacroDetail
   parameterName=OD_MD_Layer3_4_Settings
   channelNames {a=Layer4_Falloff, b=Layer4_Height, g=Layer3_Falloff, r=Layer3_Height}
   defaultValue=<rgba -1500 500 -2500 800>
  }
  n18 = Add(A=@n13, B=@n17.Layer3_Height)
  n19 = Add(A=@n13, B=@n17.Layer4_Height)
  FunctionOutput(None=MaterialFunctionCall(
    X=Divide(A=@n15, B=@n14.Layer1_Falloff)
    Y=Divide(A=@n16, B=@n14.Layer2_Falloff)
    Z=Divide(A=@n18, B=@n17.Layer3_Falloff)
    A=Divide(A=@n19, B=@n17.Layer4_Falloff)
   ) {materialFunction=&MakeFloat4}.Result) {outputName=SedimentLayers, sortPriority=1}
  n20 = DotProduct(A=VertexNormalWS(), B=Constant3Vector() {constant=<rgba 0 0 1 0>})
  n21 = Saturate(None=Multiply(A=Power(Base=@n20, Exp=@n5.Normal_SlopeMaskPower), B=@n5.Normal_SlopeMaskMod))
  n22 = Multiply(A=Saturate(None=Add(A=@n21, B=@n5.Normal_SlopeMaskAdd)), B=FunctionInput() {inputName=BiomeMask, inputType=FunctionInput_Scalar})
  FunctionOutput(None=Multiply(A=Multiply(A=@n2), B=@n22)) {outputName=Normals, sortPriority=-100}
  n23 = TextureSample(UVs=Multiply(A=TextureCoordinate(), B=Constant() {r=0.02})) {
   samplerSource=SSM_Wrap_WorldGroupSettings
   samplerType=SAMPLERTYPE_Normal
   texture=&T_Helios_DesertWaves_N
  }
  LinearInterpolate(
   A=@n23.RGB
   B=LinearInterpolate(
    A=NamedRerouteUsage()
    B=NamedRerouteUsage()
    Alpha=ScalarParameter() {defaultValue=1, group=OpenDesert_MacroDetail, parameterName=OD_MD_DEBUG_WSN}
   )
   Alpha=ScalarParameter() {defaultValue=1, group=OpenDesert_MacroDetail, parameterName=OD_MD_DEBUG_CORRECTNORMALS}
  )
  Max(A=ScalarParameter() {group=OpenDesert_MacroDetail, parameterName=OD_MD_BiomeMask_DEBUG}, B=@n22)
  n24 = VertexNormalWS()
  n25 = Constant3Vector() {constant=<rgba 0 0 1 0>}
  n26 = Multiply(
   A=If(A=DotProduct(A=@n24, B=@n25) {desc=generate sign for this projection}, B=Constant(), "A > B"=Constant() {r=1}, "A < B"=Constant() {r=-1})
   B=Constant3Vector() {constant=<rgba 1 0 0 0>}
  )
  n27 = Constant3Vector() {constant=<rgba 0 -1 1 0>}
  n28 = NamedRerouteUsage()
  n29 = MaterialFunctionCall!(Vector1=@n24, Vector2=Constant3Vector() {constant=<rgba 0 1 0 0>}) {materialFunction=/Engine/Functions/Engine_MaterialFunctions02/Math/CreateThirdOrthogonalVector}
  n30 = MaterialFunctionCall(
   VectorToTransform=Multiply(A=Add(A=@n26, B=@n27), B=Multiply(A=@n28, B=@n6.RGB))
   BasisX=@n29.Vector3
   BasisY=@n29.Vector3
   BasisZ=@n29.Vector3
  ) {
   desc=lerps in when the vertex normal faces X or Negative X
   materialFunction=&Transform3x3Matrix
  }
  NamedRerouteDeclaration(None=@n30.Transformed Vector) {name=Correct_WSNormals, nodeColor=<rgba 0.5 0 1 1>}
  NamedRerouteDeclaration(None=Transform(None=@n30.Transformed Vector) {transformSourceType=TRANSFORMSOURCE_World, transformType=TRANSFORM_Tangent}) {name=Correct_TangentNormals, nodeColor=<rgba 0.5 0 1 1>}
  NamedRerouteDeclaration(None=VectorParameter() {group=OpenDesert_MacroDetail, parameterName=OD_MD_NormalMod, defaultValue=<rgba 1 1 0.2 1>}.RGB) {name=NormalMod3Vec, nodeColor=<rgba 0.5 0 1 1>}
  ScalarParameter() {defaultValue=8500, group=OpenDesert_MacroDetail, parameterName=OD_MD_DEBUG_NormalScale}
  StaticSwitch(
   True=Subtract(A=@n7, B=@n8)
   False=@n12
   Value=FunctionInput(Preview=StaticBool()) {bUsePreviewValueAsDefault=true, inputName=IsForStaticMesh, inputType=FunctionInput_StaticBool, sortPriority=500}
  )
  n31 = NamedRerouteUsage()
  Transform(None=ComponentMask(None=Multiply(A=@n2, B=@n31)) {b=true, g=true, r=true})
 ]}
